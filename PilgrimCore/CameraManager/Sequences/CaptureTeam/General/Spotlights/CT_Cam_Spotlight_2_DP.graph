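# Level Sequence for a CaptureTeam spotlight camera shot: one animated cine camera move on a posed group of mannequins, about eleven seconds at 60 fps.
# Four fixed-position female soldier mannequins and an attached player mannequin play skeletal animations, alongside a spotLight_relocator actor at identity transform.
# A single camera cut covers the attached camera, which animates position, rotation and manual focus, with fixed lens settings and a look-at offset.
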LevelSequence /PilgrimCore/CameraManager/Sequences/CaptureTeam/General/Spotlights/CT_Cam_Spotlight_2_DP
bindings [
 spotLight_relocator {
  id=#1
  possessed_class=Actor
  tracks [
   {type=MovieSceneSpawnTrack, sections [{type=MovieSceneSpawnSection, channels [None {type=MovieSceneScriptingBoolChannel, default=true}]}]}
   {
    type=MovieScene3DTransformTrack
    sections [{
      type=MovieScene3DTransformSection
      channels [
       Location.X {type=MovieSceneScriptingDoubleChannel, default=0}
       Location.Y {type=MovieSceneScriptingDoubleChannel, default=0}
       Location.Z {type=MovieSceneScriptingDoubleChannel, default=0}
       Rotation.X {type=MovieSceneScriptingDoubleChannel, default=0}
       Rotation.Y {type=MovieSceneScriptingDoubleChannel, default=0}
       Rotation.Z {type=MovieSceneScriptingDoubleChannel, default=0}
       Scale.X {type=MovieSceneScriptingDoubleChannel, default=1}
       Scale.Y {type=MovieSceneScriptingDoubleChannel, default=1}
       Scale.Z {type=MovieSceneScriptingDoubleChannel, default=1}
      ]
     }]
   }
  ]
 }
 "SoldierMannequin_Female (3)" {
  id=#2
  possessed_class=SoldierMannequin_Female_C
  tracks [
   {type=MovieSceneSpawnTrack, sections [{type=MovieSceneSpawnSection, channels [None {type=MovieSceneScriptingBoolChannel, default=true}]}]}
   {
    type=MovieScene3DTransformTrack
    sections [{
      type=MovieScene3DTransformSection
      channels [
       Location.X {type=MovieSceneScriptingDoubleChannel, default=-7202.4811, const [-7202.481,1]}
       Location.Y {type=MovieSceneScriptingDoubleChannel, default=-230.7822, const [-230.7822,1]}
       Location.Z {type=MovieSceneScriptingDoubleChannel, default=626.5436, const [626.5436,1]}
       Rotation.X {type=MovieSceneScriptingDoubleChannel, default=0, const [0,1]}
       Rotation.Y {type=MovieSceneScriptingDoubleChannel, default=0, const [0,1]}
       Rotation.Z {type=MovieSceneScriptingDoubleChannel, default=-68.9807, const [-68.9807,1]}
       Scale.X {type=MovieSceneScriptingDoubleChannel, default=1, const [1,1]}
       Scale.Y {type=MovieSceneScriptingDoubleChannel, default=1, const [1,1]}
       Scale.Z {type=MovieSceneScriptingDoubleChannel, default=1, const [1,1]}
      ]
     }]
   }
   {type=MovieSceneSkeletalAnimationTrack, sections [{type=MovieSceneSkeletalAnimationSection, end_frame=768, start_frame=0}]}
  ]
 }
 Fort Player Mannequin {
  id=#3
  possessed_class=FortPlayerMannequin
  tracks [
   {
    type=MovieScene3DTransformTrack
    sections [{
      type=MovieScene3DTransformSection
      channels [
       Location.X {type=MovieSceneScriptingDoubleChannel, default=0}
       Location.Y {type=MovieSceneScriptingDoubleChannel, default=0}
       Location.Z {type=MovieSceneScriptingDoubleChannel, default=-90.7379}
       Rotation.X {type=MovieSceneScriptingDoubleChannel, default=0}
       Rotation.Y {type=MovieSceneScriptingDoubleChannel, default=0}
       Rotation.Z {type=MovieSceneScriptingDoubleChannel, default=-90}
       Scale.X {type=MovieSceneScriptingDoubleChannel, default=1}
       Scale.Y {type=MovieSceneScriptingDoubleChannel, default=1}
       Scale.Z {type=MovieSceneScriptingDoubleChannel, default=1}
      ]
     }]
   }
   {type=MovieSceneSkeletalAnimationTrack, sections [{type=MovieSceneSkeletalAnimationSection, end_frame=709, start_frame=0}]}
   {type=MovieScene3DAttachTrack, sections [{type=MovieScene3DAttachSection, end_frame=709, start_frame=0}]}
  ]
 }
 "SoldierMannequin_Female (2)" {
  id=#4
  possessed_class=SoldierMannequin_Female_C
  tracks [
   {type=MovieSceneSpawnTrack, sections [{type=MovieSceneSpawnSection, channels [None {type=MovieSceneScriptingBoolChannel, default=true}]}]}
   {
    type=MovieScene3DTransformTrack
    sections [{
      type=MovieScene3DTransformSection
      channels [
       Location.X {type=MovieSceneScriptingDoubleChannel, default=-7099.0414, const [-7099.0415,1]}
       Location.Y {type=MovieSceneScriptingDoubleChannel, default=-779.9541, const [-779.9542,1]}
       Location.Z {type=MovieSceneScriptingDoubleChannel, default=677.6108, const [677.6108,1]}
       Rotation.X {type=MovieSceneScriptingDoubleChannel, default=0, const [0,1]}
       Rotation.Y {type=MovieSceneScriptingDoubleChannel, default=0, const [0,1]}
       Rotation.Z {type=MovieSceneScriptingDoubleChannel, default=-62.0882, const [-62.0881,1]}
       Scale.X {type=MovieSceneScriptingDoubleChannel, default=1, const [1,1]}
       Scale.Y {type=MovieSceneScriptingDoubleChannel, default=1, const [1,1]}
       Scale.Z {type=MovieSceneScriptingDoubleChannel, default=1, const [1,1]}
      ]
     }]
   }
   {type=MovieSceneSkeletalAnimationTrack, sections [{type=MovieSceneSkeletalAnimationSection, end_frame=768, start_frame=0}]}
  ]
 }
 CineCameraActor1 {
  id=#5
  possessed_class=CineCameraActor
  tracks [
   {
    type=MovieScene3DTransformTrack
    sections [{
      type=MovieScene3DTransformSection
      channels [
       Location.X {type=MovieSceneScriptingDoubleChannel, default=-5150.5862, keys [[0,-6954.959],[319,-6992.853],[643,-6963.3745]]}
       Location.Y {type=MovieSceneScriptingDoubleChannel, default=-552.7704, keys [[0,-1.073],[319,-90.4953],[643,-109.5379]]}
       Location.Z {type=MovieSceneScriptingDoubleChannel, default=671.3617, keys [[0,677.8703],[319,706.2305],[643,699.2281]]}
       Rotation.X {type=MovieSceneScriptingDoubleChannel, default=14, keys [[0,14],[643,3]]}
       Rotation.Y {type=MovieSceneScriptingDoubleChannel, default=7.8, keys [[0,15],[166,8.25],[643,7.8]]}
       Rotation.Z {type=MovieSceneScriptingDoubleChannel, default=210.6, keys [[0,227.9487],[268,220],[643.0025,210.6]]}
       Scale.X {type=MovieSceneScriptingDoubleChannel, default=1, const [1,1]}
       Scale.Y {type=MovieSceneScriptingDoubleChannel, default=1, const [1,1]}
       Scale.Z {type=MovieSceneScriptingDoubleChannel, default=1, const [1,1]}
      ]
     }]
   }
   {type=MovieSceneSpawnTrack, sections [{type=MovieSceneSpawnSection, channels [None {type=MovieSceneScriptingBoolChannel, default=true}]}]}
   {type=MovieScene3DAttachTrack, sections [{type=MovieScene3DAttachSection, end_frame=695, start_frame=0}]}
   {
    type=MovieSceneDoubleVectorTrack
    display_name="Relative Offset (Lookat Tracking Settings)"
    sections [{
      type=MovieSceneDoubleVectorSection
      channels [
       Vector.X {type=MovieSceneScriptingDoubleChannel, default=-40}
       Vector.Y {type=MovieSceneScriptingDoubleChannel, default=60}
       Vector.Z {type=MovieSceneScriptingDoubleChannel, default=60}
      ]
     }]
   }
  ]
 }
 SoldierMannequin_Female {
  id=#6
  possessed_class=SoldierMannequin_Female_C
  tracks [
   {type=MovieSceneSpawnTrack, sections [{type=MovieSceneSpawnSection, channels [None {type=MovieSceneScriptingBoolChannel, default=true}]}]}
   {
    type=MovieScene3DTransformTrack
    sections [{
      type=MovieScene3DTransformSection
      channels [
       Location.X {type=MovieSceneScriptingDoubleChannel, default=-6893.6674, const [-6893.6675,1]}
       Location.Y {type=MovieSceneScriptingDoubleChannel, default=-470.6701, const [-470.6701,1]}
       Location.Z {type=MovieSceneScriptingDoubleChannel, default=626.9851, const [626.9851,1]}
       Rotation.X {type=MovieSceneScriptingDoubleChannel, default=0, const [0,1]}
       Rotation.Y {type=MovieSceneScriptingDoubleChannel, default=0, const [0,1]}
       Rotation.Z {type=MovieSceneScriptingDoubleChannel, default=-73.634, const [-73.6339,1]}
       Scale.X {type=MovieSceneScriptingDoubleChannel, default=1, const [1,1]}
       Scale.Y {type=MovieSceneScriptingDoubleChannel, default=1, const [1,1]}
       Scale.Z {type=MovieSceneScriptingDoubleChannel, default=1, const [1,1]}
      ]
     }]
   }
   {type=MovieSceneSkeletalAnimationTrack, sections [{type=MovieSceneSkeletalAnimationSection, end_frame=806, start_frame=0}]}
  ]
 }
 "SoldierMannequin_Female (4)" {
  id=#7
  possessed_class=SoldierMannequin_Female_C
  tracks [
   {type=MovieSceneSpawnTrack, sections [{type=MovieSceneSpawnSection, channels [None {type=MovieSceneScriptingBoolChannel, default=true}]}]}
   {
    type=MovieScene3DTransformTrack
    sections [{
      type=MovieScene3DTransformSection
      channels [
       Location.X {type=MovieSceneScriptingDoubleChannel, default=-7427.9641, const [-7427.9639,1]}
       Location.Y {type=MovieSceneScriptingDoubleChannel, default=-609.7594, const [-609.7594,1]}
       Location.Z {type=MovieSceneScriptingDoubleChannel, default=778.4106, const [778.4106,1]}
       Rotation.X {type=MovieSceneScriptingDoubleChannel, default=0, const [0,1]}
       Rotation.Y {type=MovieSceneScriptingDoubleChannel, default=0, const [0,1]}
       Rotation.Z {type=MovieSceneScriptingDoubleChannel, default=-56.714, const [-56.714,1]}
       Scale.X {type=MovieSceneScriptingDoubleChannel, default=1, const [1,1]}
       Scale.Y {type=MovieSceneScriptingDoubleChannel, default=1, const [1,1]}
       Scale.Z {type=MovieSceneScriptingDoubleChannel, default=1, const [1,1]}
      ]
     }]
   }
   {type=MovieSceneSkeletalAnimationTrack, sections [{type=MovieSceneSkeletalAnimationSection, end_frame=783, start_frame=-128}]}
  ]
 }
 CameraComponent {
  id=#8
  parent_id=#5
  possessed_class=CineCameraComponent
  tracks [
   {
    type=MovieSceneFloatTrack
    display_name=Current Focal Length
    sections [{type=MovieSceneFloatSection, channels [None {type=MovieSceneScriptingFloatChannel, default=28, const [28,1]}]}]
   }
   {
    type=MovieSceneFloatTrack
    display_name="Manual Focus Distance (Focus Settings)"
    sections [{
      type=MovieSceneFloatSection
      channels [None {
        type=MovieSceneScriptingFloatChannel
        default=231.4355
        keys [[0,761.3271],[78,292.7439],[212,268.1585],[508,238.7284],[587,231.4355],[643,270.0827]]
       }]
     }]
   }
   {
    type=MovieSceneFloatTrack
    display_name=Current Aperture
    sections [{type=MovieSceneFloatSection, channels [None {type=MovieSceneScriptingFloatChannel, default=3, const [3,1]}]}]
   }
   {type=MovieSceneCameraShakeTrack, sections []}
  ]
 }
]
display_rate {denominator=1, numerator=60}
tick_resolution {denominator=1, numerator=24000}
tracks [{type=MovieSceneCameraCutTrack, sections [{type=MovieSceneCameraCutSection, end_frame=643, start_frame=0}]}]
