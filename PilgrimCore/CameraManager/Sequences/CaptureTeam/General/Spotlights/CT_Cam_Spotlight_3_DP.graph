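# Level sequence for a Capture Team spotlight cinematic camera shot, staged around a spawned spotLight_relocator actor.
# A single camera cut covers an attached CineCamera that travels and rotates with keyed roll, a zoom-out and animated focus distance at constant aperture.
# An attached, animated Fort player mannequin shares the scene with four animated female soldier mannequins at fixed transforms, whose visibility tracks are set to false.
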LevelSequence /PilgrimCore/CameraManager/Sequences/CaptureTeam/General/Spotlights/CT_Cam_Spotlight_3_DP
bindings [
 spotLight_relocator {
  id=#1
  possessed_class=Actor
  tracks [
   {type=MovieSceneSpawnTrack, sections [{type=MovieSceneSpawnSection, channels [None {type=MovieSceneScriptingBoolChannel, default=true}]}]}
   {
    type=MovieScene3DTransformTrack
    sections [{
      type=MovieScene3DTransformSection
      channels [
       Location.X {type=MovieSceneScriptingDoubleChannel, default=0}
       Location.Y {type=MovieSceneScriptingDoubleChannel, default=0}
       Location.Z {type=MovieSceneScriptingDoubleChannel, default=0}
       Rotation.X {type=MovieSceneScriptingDoubleChannel, default=0}
       Rotation.Y {type=MovieSceneScriptingDoubleChannel, default=0}
       Rotation.Z {type=MovieSceneScriptingDoubleChannel, default=0}
       Scale.X {type=MovieSceneScriptingDoubleChannel, default=1}
       Scale.Y {type=MovieSceneScriptingDoubleChannel, default=1}
       Scale.Z {type=MovieSceneScriptingDoubleChannel, default=1}
      ]
     }]
   }
  ]
 }
 "SoldierMannequin_Female (3)" {
  id=#2
  possessed_class=SoldierMannequin_Female_C
  tracks [
   {
    type=MovieScene3DTransformTrack
    sections [{
      type=MovieScene3DTransformSection
      channels [
       Location.X {type=MovieSceneScriptingDoubleChannel, default=-7202.4811, const [-7202.481,1]}
       Location.Y {type=MovieSceneScriptingDoubleChannel, default=-230.7822, const [-230.7822,1]}
       Location.Z {type=MovieSceneScriptingDoubleChannel, default=626.5436, const [626.5436,1]}
       Rotation.X {type=MovieSceneScriptingDoubleChannel, default=0, const [0,1]}
       Rotation.Y {type=MovieSceneScriptingDoubleChannel, default=0, const [0,1]}
       Rotation.Z {type=MovieSceneScriptingDoubleChannel, default=-68.9807, const [-68.9807,1]}
       Scale.X {type=MovieSceneScriptingDoubleChannel, default=1, const [1,1]}
       Scale.Y {type=MovieSceneScriptingDoubleChannel, default=1, const [1,1]}
       Scale.Z {type=MovieSceneScriptingDoubleChannel, default=1, const [1,1]}
      ]
     }]
   }
   {type=MovieSceneSkeletalAnimationTrack, sections [{type=MovieSceneSkeletalAnimationSection, end_frame=768, start_frame=0}]}
   {
    type=MovieSceneVisibilityTrack
    sections [{type=MovieSceneVisibilitySection, channels [None {type=MovieSceneScriptingBoolChannel, default=false}]}]
   }
  ]
 }
 Fort Player Mannequin {
  id=#3
  possessed_class=FortPlayerMannequin
  tracks [
   {
    type=MovieScene3DTransformTrack
    sections [{
      type=MovieScene3DTransformSection
      channels [
       Location.X {type=MovieSceneScriptingDoubleChannel, default=0}
       Location.Y {type=MovieSceneScriptingDoubleChannel, default=0}
       Location.Z {type=MovieSceneScriptingDoubleChannel, default=-90.7379}
       Rotation.X {type=MovieSceneScriptingDoubleChannel, default=0}
       Rotation.Y {type=MovieSceneScriptingDoubleChannel, default=0}
       Rotation.Z {type=MovieSceneScriptingDoubleChannel, default=-90}
       Scale.X {type=MovieSceneScriptingDoubleChannel, default=1}
       Scale.Y {type=MovieSceneScriptingDoubleChannel, default=1}
       Scale.Z {type=MovieSceneScriptingDoubleChannel, default=1}
      ]
     }]
   }
   {type=MovieSceneSkeletalAnimationTrack, sections [{type=MovieSceneSkeletalAnimationSection, end_frame=709, start_frame=0}]}
   {type=MovieScene3DAttachTrack, sections [{type=MovieScene3DAttachSection, end_frame=709, start_frame=0}]}
  ]
 }
 "SoldierMannequin_Female (2)" {
  id=#4
  possessed_class=SoldierMannequin_Female_C
  tracks [
   {
    type=MovieScene3DTransformTrack
    sections [{
      type=MovieScene3DTransformSection
      channels [
       Location.X {type=MovieSceneScriptingDoubleChannel, default=-7099.0414, const [-7099.0415,1]}
       Location.Y {type=MovieSceneScriptingDoubleChannel, default=-779.9541, const [-779.9542,1]}
       Location.Z {type=MovieSceneScriptingDoubleChannel, default=677.6108, const [677.6108,1]}
       Rotation.X {type=MovieSceneScriptingDoubleChannel, default=0, const [0,1]}
       Rotation.Y {type=MovieSceneScriptingDoubleChannel, default=0, const [0,1]}
       Rotation.Z {type=MovieSceneScriptingDoubleChannel, default=-62.0882, const [-62.0881,1]}
       Scale.X {type=MovieSceneScriptingDoubleChannel, default=1, const [1,1]}
       Scale.Y {type=MovieSceneScriptingDoubleChannel, default=1, const [1,1]}
       Scale.Z {type=MovieSceneScriptingDoubleChannel, default=1, const [1,1]}
      ]
     }]
   }
   {type=MovieSceneSkeletalAnimationTrack, sections [{type=MovieSceneSkeletalAnimationSection, end_frame=768, start_frame=0}]}
   {
    type=MovieSceneVisibilityTrack
    sections [{type=MovieSceneVisibilitySection, channels [None {type=MovieSceneScriptingBoolChannel, default=false}]}]
   }
  ]
 }
 CineCameraActor1 {
  id=#5
  possessed_class=CineCameraActor
  tracks [
   {
    type=MovieScene3DTransformTrack
    sections [{
      type=MovieScene3DTransformSection
      channels [
       Location.X {type=MovieSceneScriptingDoubleChannel, default=-6353.5686, keys [[0,-6790.7319],[621,-6242.3228]]}
       Location.Y {type=MovieSceneScriptingDoubleChannel, default=-517.7203, keys [[0,355.2042],[621,-650.5649]]}
       Location.Z {type=MovieSceneScriptingDoubleChannel, default=667.1534, keys [[0,601.5568],[621,639.1129]]}
       Rotation.X {type=MovieSceneScriptingDoubleChannel, default=12, keys [[187,0],[339,5],[587,12]]}
       Rotation.Y {type=MovieSceneScriptingDoubleChannel, default=13, keys [[0,8.25],[162,12.9],[296,13],[468,8.7845],[633,7.4369]]}
       Rotation.Z {type=MovieSceneScriptingDoubleChannel, default=200, keys [[-8,262.3666],[302,200],[468,171.3455],[633,163.7789]]}
       Scale.X {type=MovieSceneScriptingDoubleChannel, default=1}
       Scale.Y {type=MovieSceneScriptingDoubleChannel, default=1}
       Scale.Z {type=MovieSceneScriptingDoubleChannel, default=1}
      ]
     }]
   }
   {type=MovieSceneSpawnTrack, sections [{type=MovieSceneSpawnSection, channels [None {type=MovieSceneScriptingBoolChannel, default=true}]}]}
   {type=MovieScene3DAttachTrack, sections [{type=MovieScene3DAttachSection, end_frame=695, start_frame=0}]}
   {
    type=MovieSceneDoubleVectorTrack
    display_name="Relative Offset (Lookat Tracking Settings)"
    sections [{
      type=MovieSceneDoubleVectorSection
      channels [
       Vector.X {type=MovieSceneScriptingDoubleChannel, default=-40}
       Vector.Y {type=MovieSceneScriptingDoubleChannel, default=60}
       Vector.Z {type=MovieSceneScriptingDoubleChannel, default=60}
      ]
     }]
   }
  ]
 }
 SoldierMannequin_Female {
  id=#6
  possessed_class=SoldierMannequin_Female_C
  tracks [
   {
    type=MovieScene3DTransformTrack
    sections [{
      type=MovieScene3DTransformSection
      channels [
       Location.X {type=MovieSceneScriptingDoubleChannel, default=-6893.6674, const [-6893.6675,1]}
       Location.Y {type=MovieSceneScriptingDoubleChannel, default=-470.6701, const [-470.6701,1]}
       Location.Z {type=MovieSceneScriptingDoubleChannel, default=626.9851, const [626.9851,1]}
       Rotation.X {type=MovieSceneScriptingDoubleChannel, default=0, const [0,1]}
       Rotation.Y {type=MovieSceneScriptingDoubleChannel, default=0, const [0,1]}
       Rotation.Z {type=MovieSceneScriptingDoubleChannel, default=-73.634, const [-73.6339,1]}
       Scale.X {type=MovieSceneScriptingDoubleChannel, default=1, const [1,1]}
       Scale.Y {type=MovieSceneScriptingDoubleChannel, default=1, const [1,1]}
       Scale.Z {type=MovieSceneScriptingDoubleChannel, default=1, const [1,1]}
      ]
     }]
   }
   {type=MovieSceneSkeletalAnimationTrack, sections [{type=MovieSceneSkeletalAnimationSection, end_frame=768, start_frame=0}]}
   {
    type=MovieSceneVisibilityTrack
    sections [{type=MovieSceneVisibilitySection, channels [None {type=MovieSceneScriptingBoolChannel, default=false}]}]
   }
  ]
 }
 "SoldierMannequin_Female (4)" {
  id=#7
  possessed_class=SoldierMannequin_Female_C
  tracks [
   {
    type=MovieScene3DTransformTrack
    sections [{
      type=MovieScene3DTransformSection
      channels [
       Location.X {type=MovieSceneScriptingDoubleChannel, default=-7427.9641, const [-7427.9639,1]}
       Location.Y {type=MovieSceneScriptingDoubleChannel, default=-609.7594, const [-609.7594,1]}
       Location.Z {type=MovieSceneScriptingDoubleChannel, default=778.4106, const [778.4106,1]}
       Rotation.X {type=MovieSceneScriptingDoubleChannel, default=0, const [0,1]}
       Rotation.Y {type=MovieSceneScriptingDoubleChannel, default=0, const [0,1]}
       Rotation.Z {type=MovieSceneScriptingDoubleChannel, default=-56.714, const [-56.714,1]}
       Scale.X {type=MovieSceneScriptingDoubleChannel, default=1, const [1,1]}
       Scale.Y {type=MovieSceneScriptingDoubleChannel, default=1, const [1,1]}
       Scale.Z {type=MovieSceneScriptingDoubleChannel, default=1, const [1,1]}
      ]
     }]
   }
   {type=MovieSceneSkeletalAnimationTrack, sections [{type=MovieSceneSkeletalAnimationSection, end_frame=783, start_frame=-128}]}
   {
    type=MovieSceneVisibilityTrack
    sections [{type=MovieSceneVisibilitySection, channels [None {type=MovieSceneScriptingBoolChannel, default=false}]}]
   }
  ]
 }
 CameraComponent {
  id=#8
  parent_id=#5
  possessed_class=CineCameraComponent
  tracks [
   {
    type=MovieSceneFloatTrack
    display_name=Current Focal Length
    sections [{type=MovieSceneFloatSection, channels [None {type=MovieSceneScriptingFloatChannel, default=45, keys [[0,58],[220,45]]}]}]
   }
   {
    type=MovieSceneFloatTrack
    display_name="Manual Focus Distance (Focus Settings)"
    sections [{
      type=MovieSceneFloatSection
      channels [None {type=MovieSceneScriptingFloatChannel, default=648.9717, keys [[0,819.6811],[105,648.9717],[250,460.147],[414,566.798],[642.9975,691.1172]]}]
     }]
   }
   {
    type=MovieSceneFloatTrack
    display_name=Current Aperture
    sections [{type=MovieSceneFloatSection, channels [None {type=MovieSceneScriptingFloatChannel, default=2, const [2,1]}]}]
   }
   {type=MovieSceneCameraShakeTrack, sections []}
  ]
 }
]
display_rate {denominator=1, numerator=60}
tick_resolution {denominator=1, numerator=24000}
tracks [{type=MovieSceneCameraCutTrack, sections [{type=MovieSceneCameraCutSection, end_frame=643, start_frame=0}]}]
